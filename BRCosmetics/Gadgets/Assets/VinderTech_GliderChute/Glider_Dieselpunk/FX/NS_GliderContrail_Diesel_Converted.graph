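# Niagara contrail effect for the Dieselpunk glider, apparently converted from Cascade, producing exhaust flames, smoke, sparks and light.
# A shared skeletal mesh reader alternates flame meshes, fire-to-smoke SubUV sprites and GPU sparks between two sockets, with spawn groups varying flame and smoke layers.
# Separate bone-filtered readers place spinning, expanding ring meshes at the propellers and flickering lights offset from the elevators.
# Front-end modifiers give smoke and sparks different sizes from gameplay.
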NiagaraSystem /BRCosmetics/Gadgets/Assets/VinderTech_GliderChute/Glider_Dieselpunk/FX/NS_GliderContrail_Diesel_Converted
dependencies {
 usedDataInterfaces [
  &NiagaraDataInterfaceSkeletalMesh
  &NiagaraDataInterfaceArrayFloat3
  &NiagaraDataInterfaceVectorCurve
  &NiagaraDataInterfaceCurve
  &NiagaraDataInterfaceSpriteRendererInfo
 ]
 usedDynamicInputs []
 usedModules [
  &self:SystemSpawnScript.SystemScriptSource.SystemScriptGraph.NiagaraNodeAssignment_0.SetVariables_#1
  &SystemState
 ]
}
emitters {
 FlameBig002 {
  emitterUpdateScript [
   &EmitterState(Life Cycle Mode=::NewEnumerator1, Loop Duration=0.5, UseLoopDelay=false)
   &SpawnRate(SpawnRate=5, Use Spawn Probability=false)
   &SpawnBurst_Instantaneous(Spawn Count=2, Use Spawn Probability=false, UseLoopCountLimit=false)
   &SpawnRate(SpawnGroup=1, SpawnRate=5, Use Spawn Probability=false) {module=SpawnRate001}
   &SpawnBurst_Instantaneous(Spawn Group=1, Use Spawn Probability=false, UseLoopCountLimit=false) {module=SpawnBurst_Instantaneous001}
  ]
  particleSpawnScript [
   &FN_SetSpawnGroup(Number of Spawn Groups=2)
   &InitializeParticle(
    Lifetime=0.5
    Mesh Scale=/Niagara/DynamicInputs/CascadeConversion/SplitVector(
     Double Component Vector Channels:ECascadeNiagaraTwoVectorChannels=::NewEnumerator1
     Double Component:NiagaraFloat=&RandomRangeFloat(Minimum=0.13, Maximum=0.15, Evaluation Type=::NewEnumerator1)
     Single Component:NiagaraFloat=&RandomRangeFloat(Minimum=0.075, Maximum=0.11, Evaluation Type=::NewEnumerator1)
    )
    Mesh Scale Mode=::NewEnumerator4
    UsePositionOffset=false
    Write Material Random=true
   )
   &Color(Color=&MakeLinearColorFromVectorAndFloat(
     "Vector (RGB)"=<25 5 1>
     "Float (Alpha)"=&RandomRangeFloat(Minimum=10, Maximum=15, Evaluation Type=::NewEnumerator1)
    ))
   &DynamicMaterialParameters(
    Index 0 Param 1=&MaskFloat(
     Mask Source=0
     Mask Parameter=?
     Mask=1
     Masked Float=&RandomRangeFloat(Minimum=1, Maximum=2, Evaluation Type=::NewEnumerator1)
     Passthrough Float=1
    )
    Module.Index0Float0Write=true
    Module.Index0Float1Write=false
    Module.Index0Float2Write=false
    Module.Index0Float3Write=false
   ) {module=DynamicMaterialParameters002}
   &AddVelocity(Velocity=<1 0 0>)
  ]
  particleUpdateScript [
   &ParticleState()
   &SkeletalMeshLocation(
    Bone / Socket Index=/CascadeToNiagaraConverter/NiagaraScripts/DynamicInputScripts/SequentialInt(Use Array Data:NiagaraBool=false, First Number:NiagaraInt32=0, Sequence:NiagaraDataInterfaceArrayInt32=?, Last Number:NiagaraInt32=1)
    MaskSpawnGroup=false
    Mesh Sampling Type=::NewEnumerator5
    Orientation Sampling=::NewEnumerator1
    Skeletal Mesh=$System.Skeletal Mesh Reader
    Socket Sampling Mode=::NewEnumerator5
   )
   &DynamicMaterialParameters(
    Index 0 Param 1=&MergeHelper_FloatSwitch(
     0=0
     1=&RandomRangeFloat(Minimum=0.2, Maximum=0.5, Evaluation Type=::NewEnumerator1)
     2=?
     3=?
     Index=&SpawnGroup()
     Num=::NewEnumerator0
    )
    Index 0 Param 2=1
    Index 0 Param 3=1
    Index 0 Param 4=1
    Module.Index0Float0Write=true
    Module.Index0Float1Write=true
    Module.Index0Float2Write=true
    Module.Index0Float3Write=true
   ) {module=DynamicMaterialParameters001}
   &SolveForcesAndVelocity(Clamp Velocity=false, Limit Acceleration=false)
   &CameraOffset(Camera Offset Amount=2)
   &ScaleColor(
    Scale Alpha=&FloatFromCurve(FloatCurve=DataInterface() {propertyValues {Curve {keys [
         <at 0.0945 0 RCIM_Cubic>
         {arriveTangent=1.2414, interpMode=RCIM_Cubic, leaveTangent=1.2414, time=0.2331, value=0.6716}
         {arriveTangent=-0.8758, interpMode=RCIM_Cubic, leaveTangent=-0.8758, time=0.9, value=1}
         <at 1 0 RCIM_Cubic>
        ]}}})
    Scale RGB=&VectorFromCurve(VectorCurve=DataInterface() {propertyValues {XCurve {keys [<at 0 3>,<at 1 3>]}, YCurve {keys [<at 0 0.5>,<at 1 0.5>]}, ZCurve {keys [<at 0 0.2>,<at 1 0.2>]}}})
    ScaleA=true
    ScaleRGB=true
   )
  ]
  rendererClasses [&NiagaraMeshRendererProperties]
  renderers [{rendererClass=&NiagaraMeshRendererProperties, rendererIndex=0}]
 }
 Light {
  emitterUpdateScript [
   &EmitterState(Life Cycle Mode=::NewEnumerator1, Loop Behavior=::NewEnumerator1, Loop Duration Mode=::NewEnumerator1, UseLoopDelay=false)
   &SpawnBurst_Instantaneous(Spawn Count=2, Use Spawn Probability=false, UseLoopCountLimit=false)
   &SpawnRate(SpawnRate=4, Use Spawn Probability=false) {enabled=false}
  ]
  particleSpawnScript [
   /SetVariables(Particles.SkeletonSamplingIndex:NiagaraInt32=&SequentialInt(Use Array Data:NiagaraBool=false, First Number:NiagaraInt32=0, Sequence:NiagaraDataInterfaceArrayInt32=?, Last Number:NiagaraInt32=1)) {isSetParameters=true}
   &InitializeParticle(
    Lifetime=&RandomRangeFloat(Minimum=0.55, Maximum=0.6, Evaluation Type=::NewEnumerator1)
    Sprite Size Mode=::NewEnumerator4
    UsePositionOffset=false
    Write Material Random=true
   )
   &SkeletalMeshLocation(
    Bone / Socket Index=$Particles.SkeletonSamplingIndex
    Bone Sampling Mode=::NewEnumerator4
    MaskSpawnGroup=false
    Orientation Sampling=::NewEnumerator1
    Sampled Position Offset=&SelectVectorFromArray(
     Direct Array Index=$Particles.SkeletonSamplingIndex
     Vector Selection Array=DataInterface() {propertyValues {FloatData [<-100 2 -25>, <100 -2 25>]}}
    )
    Skeletal Mesh=DataInterface() {propertyValues {
      DefaultMesh=null
      ExcludeBoneName=root
      OptionalComponentClass=null
      PreviewMesh=/BRCosmetics/Gadgets/Assets/VinderTech_GliderChute/Glider_Dieselpunk/Meshes/Glider_Dieselpunk_Export
      SkinningMode=SkinOnTheFly
      SoftSourceActor=null
      SourceMode=Default
      UvSetIndex=0
      WholeMeshLOD=-1
      bExcludeBone=false
      bReadDeformedGeometry=true
      bRequireCurrentFrameData=true
      ComponentTags []
      FilteredBones [elevator_l,elevator_r]
      FilteredSockets []
      MeshUserParameter {parameter {name=None, typeDefHandle {registeredTypeIndex=1}, varData []}}
      SamplingRegions []
     }}
   )
   &Color(Color=&MakeLinearColorFromVectorAndFloat("Vector (RGB)"=<15 1.2986 0>, "Float (Alpha)"=1))
  ]
  particleUpdateScript [
   /SetVariables(Particles.LightExponent=16) {isSetParameters=true}
   &ParticleState(Kill Particles When Lifetime Has Elapsed=false, Loop Particles Lifetime=true)
   &CascadeConversion_LightProperties(LightBrightnessScale=32, LightColorScale=<1 1 1>, LightRadiusScale=15)
   &ScaleColor(
    Scale Alpha=&FloatFromCurve(FloatCurve=DataInterface() {propertyValues {Curve {keys [
         <at 0.0013 1.0052 RCIM_Cubic>
         {arriveTangent=-1.9241, interpMode=RCIM_Cubic, leaveTangent=-1.9241, time=0.0982, value=0.8299}
         {arriveTangent=0.5687, interpMode=RCIM_Cubic, leaveTangent=0.5687, time=0.2186, value=0.5872}
         {arriveTangent=0.0271, interpMode=RCIM_Cubic, leaveTangent=0.0271, time=0.3406, value=0.9677}
         {arriveTangent=-0.0852, interpMode=RCIM_Cubic, leaveTangent=-0.0852, time=0.3978, value=0.5921}
         {arriveTangent=0.4075, interpMode=RCIM_Cubic, leaveTangent=0.4075, time=0.6589, value=0.9406}
         {arriveTangent=0.1331, interpMode=RCIM_Cubic, leaveTangent=0.1331, time=0.9182, value=0.8041}
         <at 0.9987 0.9858 RCIM_Cubic>
        ]}}})
    Scale RGB=&VectorFromCurve(VectorCurve=DataInterface() {propertyValues {XCurve {keys [<at 0 1>,<at 1 1>]}, YCurve {keys [<at 0 1>,<at 1 1>]}, ZCurve {keys [<at 0 1>,<at 1 1>]}}})
    ScaleA=true
    ScaleRGB=true
   )
  ]
  rendererClasses [&NiagaraLightRendererProperties]
  renderers [{rendererClass=&NiagaraLightRendererProperties, rendererIndex=0}]
 }
 Minimal002 {
  emitterUpdateScript [
   &EmitterState(Delay First Loop Only=true, Life Cycle Mode=::NewEnumerator1, Loop Delay=0.5, Loop Duration=0.25, UseLoopDelay=true)
   &SpawnRate(SpawnRate=65, Use Spawn Probability=false)
   &SpawnRate(SpawnGroup=1, SpawnRate=65, Use Spawn Probability=false) {module=SpawnRate001}
  ]
  particleSpawnScript [
   /SetVariables(Particles.VisibilityTag=&SpawnGroup()) {isSetParameters=true}
   &InitializeParticle(
    Lifetime=&RandomRangeFloat(Minimum=0.7, Maximum=0.7, Evaluation Type=::NewEnumerator1)
    Sprite Rotation Mode=::NewEnumerator2
    Sprite Size Mode=::NewEnumerator3
    Uniform Sprite Size=&FN_FrontEndFloatModifier(
     Gameplay=&RandomRangeFloat(Minimum=50, Maximum=90)
     Front End=&RandomRangeFloat(Minimum=30, Maximum=40)
    )
    UsePositionOffset=false
    Write Material Random=true
   )
   &Color(Color=&MergeHelper_ColorSwitch(
     0=&MakeLinearColorFromVectorAndFloat(
      "Vector (RGB)"=<0.125 0.125 0.125>
      "Float (Alpha)"=&FloatFromCurve(FloatCurve=DataInterface() {propertyValues {Curve {keys [<at 0 2>,<at 1 2>]}}})
     )
     1=&MakeLinearColorFromVectorAndFloat(
      "Vector (RGB)"=<0.125 0.1218 0.1192>
      "Float (Alpha)"=&FloatFromCurve(FloatCurve=DataInterface() {propertyValues {Curve {keys [<at 0 0.5>,<at 0.15 0.5>]}}}, Scale Curve=0.5)
     )
     2=?
     3=?
     Index=&SpawnGroup()
     Num=::NewEnumerator0
    ))
   &ShapeLocation(MaskSpawnGroup=false, Override Seed=false, Sphere Radius=5, UseOwnerScale=true)
   &InheritVelocity(Inherited Velocity Amount Scale=<-0.05 -0.05 -0.2>, Inherited Velocity Speed Limit=0, Speed Limit=false, UseSpeedThreshold=false)
   &AddVelocity(
    Cone Angle=&RandomRangeFloat(Minimum=0, Maximum=0.01, Evaluation Type=::NewEnumerator1)
    Cone Angle Mode:ENiagara_AngleInput=::NewEnumerator1
    Module.UseConeFalloff=false
    Module.UseVelDistribution=false
    Override Seed=false
    Velocity Mode=::NewEnumerator2
    Velocity Speed=&RandomRangeFloat(Minimum=400, Maximum=450, Evaluation Type=::NewEnumerator1)
   )
   &SkeletalMeshLocation(
    Bone / Socket Index=&SequentialInt(Use Array Data:NiagaraBool=false, First Number:NiagaraInt32=0, Sequence:NiagaraDataInterfaceArrayInt32=?, Last Number:NiagaraInt32=1)
    MaskSpawnGroup=false
    Mesh Sampling Type=::NewEnumerator5
    Orientation Sampling=::NewEnumerator1
    Sampled Position Offset=&MergeHelper_VectorSwitch(0=<-20 0 0>, 1=<-17 0 0>, 2=?, 3=?, Index=&SpawnGroup(), Num=::NewEnumerator0)
    Sampled Position Offset Coordinate Space:ENiagaraSkelMeshTransforms=::NewEnumerator2
    Skeletal Mesh=$System.Skeletal Mesh Reader
    Socket Sampling Mode=::NewEnumerator5
   )
  ]
  particleUpdateScript [
   &ParticleState()
   &ScaleSpriteSize(Uniform Curve Sprite Scale=DataInterface() {propertyValues {Curve {
       defaultValue=0
       keys [
        {interpMode=RCIM_Cubic, tangentMode=RCTM_User, time=0, value=0}
        {arriveTangent=1.1029, interpMode=RCIM_Cubic, leaveTangent=1.1029, tangentMode=RCTM_User, time=0.4, value=1}
       ]
      }}})
   &ScaleColor(
    Scale Alpha=&FloatFromCurve(FloatCurve=DataInterface() {propertyValues {Curve {keys [
         <at -0.0015 0.0075 RCIM_Cubic>
         {arriveTangent=2.7107, interpMode=RCIM_Cubic, leaveTangent=2.7107, time=0.1441, value=0.5874}
         {arriveTangent=0.5845, interpMode=RCIM_Cubic, leaveTangent=0.5845, time=0.3637, value=0.9974}
         {arriveTangent=-1.5676, interpMode=RCIM_Cubic, leaveTangent=-1.5676, time=0.85, value=1}
         <at 1 0 RCIM_Cubic>
        ]}}})
    Scale RGB=&MergeHelper_VectorSwitch(
     0=&VectorFromCurve(VectorCurve=DataInterface() {propertyValues {
        XCurve {keys [<at 0.15 50 RCIM_Cubic>, {arriveTangent=-142.1429, interpMode=RCIM_Cubic, leaveTangent=-142.1429, time=0.3, value=15}, <at 0.5 0.25>]}
        YCurve {keys [<at 0.15 10 RCIM_Cubic>, {arriveTangent=-27.8571, interpMode=RCIM_Cubic, leaveTangent=-27.8571, time=0.3, value=5}, <at 0.5 0.25>]}
        ZCurve {keys [<at 0.15 2 RCIM_Cubic>, {arriveTangent=-5, interpMode=RCIM_Cubic, leaveTangent=-5, time=0.3, value=0}, <at 0.5 0.25>]}
       }})
     1=&VectorFromCurve(VectorCurve=DataInterface() {propertyValues {
        XCurve {keys [<at 0.15 50 RCIM_Cubic>, {arriveTangent=-142.1429, interpMode=RCIM_Cubic, leaveTangent=-142.1429, time=0.25, value=15}, <at 0.5 0.25>]}
        YCurve {keys [<at 0.15 10 RCIM_Cubic>, {arriveTangent=-27.8571, interpMode=RCIM_Cubic, leaveTangent=-27.8571, time=0.25, value=5}, <at 0.5 0.25>]}
        ZCurve {keys [<at 0.15 2 RCIM_Cubic>, {arriveTangent=-5, interpMode=RCIM_Cubic, leaveTangent=-5, time=0.25, value=0}, <at 0.5 0.25>]}
       }})
     2=?
     3=?
     Index=&SpawnGroup()
     Num=::NewEnumerator0
    )
    ScaleA=true
    ScaleRGB=true
   )
   &Drag(
    Drag=&MergeHelper_FloatSwitch(
     0=&RandomRangeFloat(Minimum=2, Maximum=4)
     1=1
     2=?
     3=?
     Index=&SpawnGroup()
     Num=::NewEnumerator0
    )
    Use Linear Drag=true
    Use Rotational Drag=true
   )
   &DynamicMaterialParameters(
    Index 0 Param 1=&FloatFromCurve(FloatCurve=DataInterface() {propertyValues {Curve {keys [<at 0 1 RCIM_Cubic>,<at 0.9984 -1.0007 RCIM_Cubic>]}}})
    Index 0 Param 2=1
    Index 0 Param 3=1
    Index 0 Param 4=&MergeHelper_FloatSwitch(0=1, 1=0.5, 2=?, 3=?, Index=&SpawnGroup(), Num=::NewEnumerator0)
    Module.Index0Float0Write=true
    Module.Index0Float1Write=true
    Module.Index0Float2Write=true
    Module.Index0Float3Write=true
   )
   &SolveForcesAndVelocity(Clamp Velocity=false, Limit Acceleration=false)
   &SubUVAnimation(
    Frames Per Second=25
    Playback Mode:ENiagara_FPSPlayrate=::NewEnumerator1
    Playback Scale Curve=DataInterface() {propertyValues {Curve {keys [
        {interpMode=RCIM_Cubic, tangentMode=RCTM_User, time=0.25, value=1}
        {arriveTangent=-1.2656, interpMode=RCIM_Cubic, leaveTangent=-1.2656, tangentMode=RCTM_User, time=1, value=0}
       ]}}}
    "Sprite Renderer "=DataInterface() {propertyValues {SpriteRenderer=&self:Minimal002_0.NiagaraSpriteRendererProperties_0}}
    SubUV Animation Mode:ENiagara_SubUVLookupModeV2=::NewEnumerator3
    UseEndFrame=false
    UsePlayRateScale=false
    UseStartFrame=false
   )
   &CameraOffset(Camera Offset Amount=25)
  ]
  rendererClasses [&NiagaraSpriteRendererProperties]
  renderers [
   {rendererClass=&NiagaraSpriteRendererProperties}
   {rendererClass=&NiagaraSpriteRendererProperties, rendererIndex=1}
  ]
 }
 Particle_Emitter001 {
  simTarget=GPUComputeSim
  emitterUpdateScript [
   &EmitterState(
    Inactive Response=::NewEnumerator1
    Life Cycle Mode=::NewEnumerator1
    Loop Delay=&RandomRangeFloat(Minimum=0.15, Maximum=0.35, Evaluation Type=::NewEnumerator1)
    Loop Duration=2
    UseLoopDelay=true
   )
   &SpawnRate(SpawnRate=35, Use Spawn Probability=false)
  ]
  particleSpawnScript [
   &InitializeParticle(
    Lifetime=0.7
    Sprite Size=&FN_FrontEndVector2Modifier(
     Front End:Vector2f=&RandomRangeVector2D(Minimum=<0.7 1.6>, Maximum=<0.85 1.9>, Evaluation Type=::NewEnumerator1)
     Gameplay:Vector2f=&RandomRangeVector2D(Minimum=<1 3>, Maximum=<2 5>, Evaluation Type=::NewEnumerator1)
    )
    Sprite Size Mode=::NewEnumerator4
    UsePositionOffset=false
    Write Material Random=true
   )
   &SkeletalMeshLocation(
    Bone / Socket Index=&SequentialInt(Use Array Data:NiagaraBool=false, First Number:NiagaraInt32=0, Sequence:NiagaraDataInterfaceArrayInt32=?, Last Number:NiagaraInt32=1)
    MaskSpawnGroup=false
    Mesh Sampling Type=::NewEnumerator5
    Orientation Sampling=::NewEnumerator1
    Skeletal Mesh=$System.Skeletal Mesh Reader
    Socket Sampling Mode=::NewEnumerator5
   ) {module=SkeletalMeshLocation001}
   &AddVelocity(Velocity=&RandomRangeVector(Minimum=<0 0 50>, Maximum=<0 0 0>, Evaluation Type=::NewEnumerator1))
   &Color(Color=&MakeLinearColorFromVectorAndFloat("Vector (RGB)"=<5 1.7766 0.2561>, "Float (Alpha)"=0.5))
   &AddVelocity(
    Cone Angle=&RandomRangeFloat(Minimum=0.05, Maximum=0.12, Evaluation Type=::NewEnumerator1)
    Cone Angle Mode:ENiagara_AngleInput=::NewEnumerator1
    Module.UseConeFalloff=false
    Module.UseVelDistribution=false
    Override Seed=false
    Velocity Mode=::NewEnumerator2
    Velocity Speed=&RandomRangeFloat(Minimum=300, Maximum=250, Evaluation Type=::NewEnumerator1)
   ) {module=AddVelocity001}
   &ShapeLocation(MaskSpawnGroup=false, Override Seed=false, Sphere Radius=5, UseOwnerScale=true)
  ]
  particleUpdateScript [
   &ParticleState()
   &SolveForcesAndVelocity(Clamp Velocity=false, Limit Acceleration=false)
  ]
  rendererClasses [&NiagaraSpriteRendererProperties]
  renderers [{rendererClass=&NiagaraSpriteRendererProperties}]
 }
 Ring1 {
  emitterUpdateScript [
   &EmitterState(Life Cycle Mode=::NewEnumerator1, Loop Duration=1, UseLoopDelay=true)
   &SpawnRate(SpawnRate=50, Use Spawn Probability=false)
  ]
  particleSpawnScript [
   /SetVariables(Particles.MeshRotationEuler:Vector3f=&RandomRangeVector(Minimum=<0.25 0 0.25>, Maximum=<0.25 1 0.25>, Evaluation Type=::NewEnumerator1)) {isSetParameters=true}
   /SetVariables(Particles.MeshOrientation=/CascadeToNiagaraConverter/NiagaraScripts/DynamicInputScripts/MakeQuatFromEuler(Euler:Vector3f=$Particles.MeshRotationEuler, Scale:NiagaraFloat=360)) {isSetParameters=true}
   /SetVariables(Particles.MeshRotationRateBase:Vector3f=&RandomRangeVector(Minimum=<0 0.7 0>, Maximum=<0 1 0>, Evaluation Type=::NewEnumerator1)) {isSetParameters=true}
   /SetVariables(Particles.SkeletonSamplingIndex:NiagaraInt32=&SequentialInt(Use Array Data:NiagaraBool=false, First Number:NiagaraInt32=0, Sequence:NiagaraDataInterfaceArrayInt32=?, Last Number:NiagaraInt32=1)) {isSetParameters=true}
   &InitializeParticle(
    Lifetime=&RandomRangeFloat(Minimum=0.35, Maximum=0.6, Evaluation Type=::NewEnumerator1)
    Mesh Scale=&SplitVector(
     Double Component Vector Channels:ECascadeNiagaraTwoVectorChannels=::NewEnumerator0
     Double Component:NiagaraFloat=&RandomRangeFloat(Minimum=1.25, Maximum=2, Evaluation Type=::NewEnumerator1)
     Single Component:NiagaraFloat=&RandomRangeFloat(Minimum=1, Maximum=3, Evaluation Type=::NewEnumerator1)
    )
    Mesh Scale Mode=::NewEnumerator4
    UsePositionOffset=false
    Write Material Random=true
   )
   /CascadeToNiagaraConverter/NiagaraScripts/ModuleScripts/AddVectorToPosition(Vector:Vector3f=&TransformVector(Vector=<0 -20 -25>, Destination Space=::Simulation))
   &DynamicMaterialParameters(
    Index 0 Param 1=&RandomRangeFloat(Minimum=1, Maximum=2, Evaluation Type=::NewEnumerator1)
    Module.Param0WriteEnabled=true
    Module.Param1WriteEnabled=false
    Module.Param2WriteEnabled=false
    Module.Param3WriteEnabled=false
   )
   &Color(Color=&MakeLinearColorFromVectorAndFloat(
     "Vector (RGB)"=<0.05 0.05 0.05>
     "Float (Alpha)"=&RandomRangeFloat(Minimum=0.45, Maximum=0.35, Evaluation Type=::NewEnumerator1)
    ))
   &SkeletalMeshLocation(
    Bone / Socket Index=$Particles.SkeletonSamplingIndex
    Bone Sampling Mode=::NewEnumerator4
    MaskSpawnGroup=false
    Orientation Sampling=::NewEnumerator1
    Sampled Position Offset=&SelectVectorFromArray(
     Direct Array Index=$Particles.SkeletonSamplingIndex
     Vector Selection Array=DataInterface() {propertyValues {FloatData [<15 0 0>, <15 0 0>]}}
    )
    Skeletal Mesh=DataInterface() {propertyValues {
      DefaultMesh=null
      ExcludeBoneName=root
      OptionalComponentClass=null
      PreviewMesh=&Glider_Dieselpunk_Export
      SkinningMode=SkinOnTheFly
      SoftSourceActor=null
      SourceMode=Default
      UvSetIndex=0
      WholeMeshLOD=-1
      bExcludeBone=false
      bReadDeformedGeometry=true
      bRequireCurrentFrameData=true
      ComponentTags []
      FilteredBones [propeller_l,propeller_r]
      FilteredSockets []
      MeshUserParameter {parameter {name=None, typeDefHandle {registeredTypeIndex=1}, varData []}}
      SamplingRegions []
     }}
   )
   &AddVelocity(Velocity=&RandomRangeVector(Minimum=<15 0 0>, Maximum=<100 0 0>, Evaluation Type=::NewEnumerator1))
  ]
  particleUpdateScript [
   /SetVariables(Particles.MeshRotationRate:Vector3f=$Particles.MeshRotationRateBase) {isSetParameters=true}
   &ParticleState()
   &ScaleMeshSize(Scale Factor=&VectorFromCurve(VectorCurve=DataInterface() {propertyValues {XCurve {keys [<at 0 0.6>,<at 1 3>]}, YCurve {keys [<at 0 1>,<at 1 3>]}, ZCurve {keys [<at 0 1>,<at 1 5>]}}}))
   /CascadeToNiagaraConverter/NiagaraScripts/ModuleScripts/UpdateMeshOrientationByEuler(RotationRate:Vector3f=$Particles.MeshRotationRate)
   &DynamicMaterialParameters(
    Index 0 Param 2=0.2
    Index 0 Param 3=1
    Index 0 Param 4=1
    Module.Param0WriteEnabled=false
    Module.Param1WriteEnabled=true
    Module.Param2WriteEnabled=true
    Module.Param3WriteEnabled=true
   ) {module=DynamicMaterialParameters001}
   &ScaleColor(
    Scale Alpha=&FloatFromCurve(FloatCurve=DataInterface() {propertyValues {Curve {keys [<at 0.1473 0 RCIM_Cubic>,<at 0.1983 1 RCIM_Cubic>,<at 1 0 RCIM_Cubic>]}}})
    Scale RGB=&VectorFromCurve(VectorCurve=DataInterface() {propertyValues {XCurve {keys [<at 0 0>,<at 0.25 1>,<at 1 0>]}, YCurve {keys [<at 0 0>,<at 0.25 1>,<at 1 0>]}, ZCurve {keys [<at 0 0>,<at 0.25 1>,<at 1 0>]}}})
    ScaleA=true
    ScaleRGB=true
   )
   &SolveForcesAndVelocity(Clamp Velocity=false, Limit Acceleration=false)
  ]
  rendererClasses [&NiagaraMeshRendererProperties]
  renderers [{rendererClass=&NiagaraMeshRendererProperties, rendererIndex=0}]
 }
}
summary {
 systemName=NS_GliderContrail_Diesel_Converted
 emitters [
  {emitterName=Light, rendererClasses [&NiagaraLightRendererProperties]}
  {emitterName=Particle_Emitter001, simTarget=GPUComputeSim, rendererClasses [&NiagaraSpriteRendererProperties]}
  {emitterName=Ring1, rendererClasses [&NiagaraMeshRendererProperties]}
  {emitterName=FlameBig002, rendererClasses [&NiagaraMeshRendererProperties]}
  {emitterName=Minimal002, rendererClasses [&NiagaraSpriteRendererProperties]}
 ]
 userVariables [
  User.bIsFrontEnd {type=NiagaraBool}
  User.bIsFrontEndPreview {type=NiagaraBool}
 ]
}
